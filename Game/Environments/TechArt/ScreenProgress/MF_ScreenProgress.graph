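# Material function that adds an emissive capture-progress screen to incoming material attributes, with a static switch to pass them through unchanged.
# MPC_Hologram supplies progress, color switch, captured, contested, capture animation and screen-enable states.
# It builds a noise-waved progress line, animated bar, stripes, vignette, logo and glitch mask, with paired color sets, contested blink and capture flash.
# GIReplace boosts indirect emissive for Lumen; another switch disables the logo on characters.
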MaterialFunction /Game/Environments/TechArt/ScreenProgress/MF_ScreenProgress
graph {nodes [
  Add()
  n1 = Subtract(A=ComponentMask(None=NamedRerouteUsage()) {g=true}, B=NamedRerouteUsage())
  n2 = Multiply(A=NamedRerouteUsage(), B=ScalarParameter() {defaultValue=0.308, parameterName=MusicWaveNoise_1})
  n3 = Time()
  n4 = Frac(None=Multiply(A=@n3, B=ScalarParameter() {defaultValue=-0.044, parameterName=WavePan_Speed}))
  n5 = ScalarParameter() {defaultValue=0.036, parameterName=WavePan_Speed2}
  n6 = AppendVector(A=ComponentMask(None=Add(A=@n2, B=@n4)) {r=true}, B=Frac(None=Multiply(A=@n3, B=@n5)))
  n7 = Power(
   Base=Add(
    A=TextureSample(UVs=@n6) {texture=/Game/Global/Textures/Default/Noise/T_TileNoise_02}.R
    B=ScalarParameter() {defaultValue=0.9757, parameterName=WaveNoiseAdd}
   )
   Exp=ScalarParameter() {defaultValue=2.7397, parameterName=WaveNoiseFalloff1}
  )
  Add(A=@n1, B=Multiply(A=@n7, B=ScalarParameter() {defaultValue=1, parameterName=WaveNoise_AmpProg}))
  n8 = FunctionInput(Preview=SetMaterialAttributes!(MaterialAttributes)) {bUsePreviewValueAsDefault=true, inputType=FunctionInput_MaterialAttributes}
  n9 = GetMaterialAttributes(None=@n8) {attributeGetTypes [#1, #2]}
  n10 = FunctionInput(Preview=Constant() {r=1}) {bUsePreviewValueAsDefault=true, inputName=Mask, inputType=FunctionInput_Scalar, sortPriority=6, previewValue=<1 0 0 1>}
  n11 = Multiply(A=NamedRerouteUsage(), B=OneMinus(None=NamedRerouteUsage()))
  n12 = NamedRerouteUsage()
  n13 = LinearInterpolate(
   A=VectorParameter() {parameterName=Color_Progress_A, defaultValue=<rgba 0 0.3585 1 1>}.RGB
   B=VectorParameter() {parameterName=Color_Progress, defaultValue=<rgba 1 0.9142 0 1>}.RGB
   Alpha=NamedRerouteUsage()
  )
  n14 = NamedRerouteUsage()
  n15 = LinearInterpolate(
   A=LinearInterpolate(
    A=VectorParameter() {parameterName=Color_Vignette_1_A, defaultValue=<rgba 0.0215 0.0732 0.1458 1>}.RGB
    B=VectorParameter() {parameterName=Color_Vignette_1, defaultValue=<rgba 0.1458 0.1051 0 1>}.RGB
    Alpha=NamedRerouteUsage()
   )
   B=LinearInterpolate(
    A=VectorParameter() {parameterName=Color_Vignette_2_B, defaultValue=<rgba 0.0266 0.0592 0.1198 1>}.RGB
    B=VectorParameter() {parameterName=Color_Vignette_2, defaultValue=<rgba 0.1195 0.0976 0 1>}.RGB
    Alpha=NamedRerouteUsage()
   )
   Alpha=Add(A=NamedRerouteUsage(), B=NamedRerouteUsage())
  )
  n16 = LinearInterpolate(
   A=VectorParameter() {parameterName=Color_Logo_A, defaultValue=<rgba 0.025 0.1876 0.4844 1>}.RGB
   B=VectorParameter() {parameterName=Color_Logo, defaultValue=<rgba 0.3594 0.278 0 1>}.RGB
   Alpha=NamedRerouteUsage()
  )
  n17 = LinearInterpolate(
   A=VectorParameter() {parameterName=Color_Logo2_A, defaultValue=<rgba 0 0.0287 0.0729 1>}.RGB
   B=VectorParameter() {parameterName=Color_Logo2, defaultValue=<rgba 0.0312 0.024 0 1>}.RGB
   Alpha=NamedRerouteUsage()
  )
  n18 = NamedRerouteUsage()
  n19 = Add(A=@n15, B=Multiply(A=LinearInterpolate(A=@n16, B=@n17, Alpha=@n18), B=NamedRerouteUsage()))
  n20 = LinearInterpolate(
   A=LinearInterpolate(
    A=VectorParameter() {parameterName=Color_A, defaultValue=<rgba 0 0.0143 0.0469 1>}.RGB
    B=VectorParameter() {parameterName=Color, defaultValue=<rgba 0.0469 0.0384 0 1>}.RGB
    Alpha=NamedRerouteUsage()
   )
   B=@n13
   Alpha=NamedRerouteUsage()
  )
  n21 = NamedRerouteUsage()
  n22 = Add(A=Multiply(A=Add(A=@n11, B=@n12), B=@n13), B=Add(A=Multiply(A=@n14, B=@n19), B=Multiply(A=@n20, B=@n21)))
  n23 = Multiply(A=@n22) {constB=0.1}
  n24 = Saturate(None=@n22)
  n25 = Multiply(A=Time(), B=ScalarParameter() {defaultValue=1.5, parameterName=BlinkSpeed})
  n26 = Divide(A=Add(A=Sine(None=@n25)))
  n27 = OneMinus(None=Multiply(A=Add(A=@n26) {constB=0.5}) {constB=0.5})
  n28 = NamedRerouteUsage()
  n29 = Multiply(
   A=Desaturation(None=Add(A=@n23, B=@n24), Fraction=Constant() {r=1})
   B=Multiply(A=VectorParameter() {parameterName=Color Contested, defaultValue=<rgba 3 0 0 1>}.RGB, B=Multiply(A=@n27, B=@n28))
  )
  n30 = NamedRerouteUsage()
  n31 = VectorParameter() {parameterName=ColorCaptureFlash, defaultValue=<rgba 9.388 10 0 1>}
  n32 = Add(A=Add(A=NamedRerouteUsage(), B=NamedRerouteUsage())) {constB=0.15}
  n33 = LinearInterpolate(
   A=Multiply(A=@n10, B=LinearInterpolate(A=@n22, B=@n29, Alpha=@n30) {constAlpha=1})
   B=Multiply(A=@n31.RGB, B=Saturate(None=@n32))
   Alpha=NamedRerouteUsage()
  )
  n34 = Multiply(A=@n33, B=ScalarParameter() {defaultValue=1, parameterName=LumenBoost})
  n35 = Multiply(A=Add(A=@n9.EmissiveColor, B=GIReplace!(Default=@n33, StaticIndirect=@n34, DynamicIndirect=@n34)), B=NamedRerouteUsage())
  n36 = LinearInterpolate(A=@n9.BaseColor, B=Multiply(A=@n9.BaseColor, B=NamedRerouteUsage()), Alpha=NamedRerouteUsage())
  n37 = FunctionInput(Preview=StaticBool() {value=true}) {bUsePreviewValueAsDefault=true, inputName=IsEnabled, inputType=FunctionInput_StaticBool, sortPriority=66, previewValue=<1 0 0 1>}
  FunctionOutput(None=StaticSwitch(
    True=SetMaterialAttributes!(MaterialAttributes=@n9.MaterialAttributes, EmissiveColor=@n35, BaseColor=@n36) {attributeSetTypes [#1, #2]}
    False=@n8
    Value=@n37
   ))
  LinearInterpolate(A=VectorParameter() {parameterName=ColorCaptureFlash_A, defaultValue=<rgba 0 150.7837 500 1>}.RGB, B=@n31.RGB, Alpha=NamedRerouteUsage())
  n38 = NamedRerouteUsage()
  LinearInterpolate(A=@n33, B=Multiply(A=@n33, B=OneMinus(None=@n38)), Alpha=NamedRerouteUsage())
  n39 = Time()
  n40 = ScalarParameter() {defaultValue=0.032, parameterName=LogoSpeed_2}
  LinearInterpolate(Alpha=Sine(None=Multiply(A=@n39, B=@n40))) {constA=1.1, constB=1.2}
  Multiply(A=@n15, B=OneMinus(None=NamedRerouteUsage()))
  NamedRerouteDeclaration(None=FunctionInput(Preview=TextureCoordinate()) {bUsePreviewValueAsDefault=true, inputName=UVs, inputType=FunctionInput_Vector2, sortPriority=2}) {name=UVs, nodeColor=<rgba 0 1 0.25 1>}
  n41 = NamedRerouteDeclaration(None=Saturate(None=CollectionParameter() {collection=/Game/Environments/TechArt/Hologram/MPC_Hologram, parameterName=ProgressNormalized})) {name=ProgressNormalized, nodeColor=<rgba 0.6172 0 1 1>}
  n42 = Multiply(A=NamedRerouteUsage()) {constB=2}
  NamedRerouteDeclaration(None=OneMinus(None=Add(A=@n41, B=@n42))) {name=ProgressNormalized_inverted, nodeColor=<rgba 0.6172 0 1 1>}
  n43 = ScalarParameter() {defaultValue=0.0189, parameterName=ProgressLine_TopLine}
  n44 = ScalarParameter() {defaultValue=-2.3814, parameterName=WaveNoise_Amp1}
  n45 = Add(A=@n1, B=Multiply(A=@n1, B=Multiply(A=@n7, B=@n44)))
  n46 = Saturate(None=Abs(None=Divide(A=@n45, B=@n43)))
  NamedRerouteDeclaration(None=Multiply(A=OneMinus(None=@n46), B=ScalarParameter() {defaultValue=3, parameterName=ProgressLine_TopLine_Brightness})) {name=Progress Line Only, nodeColor=<rgba 0 0.4844 1 1>}
  n47 = NamedRerouteUsage()
  n48 = Multiply(A=ComponentMask(None=@n47) {g=true}, B=ScalarParameter() {defaultValue=47.6768, parameterName=Strips})
  NamedRerouteDeclaration(None=Saturate(None=Sine(None=@n48))) {name=Stripes, nodeColor=<rgba 1 0 0.3281 1>}
  n49 = Divide(A=Distance(A=@n47, B=Constant2Vector() {g=0.5, r=0.5}), B=ScalarParameter() {defaultValue=1.3405, parameterName=Vignette})
  NamedRerouteDeclaration(None=Power(Base=@n49, Exp=ScalarParameter() {defaultValue=1.9402, parameterName=Vignette_Falloff})) {name=Vignette, nodeColor=<rgba 0 0.6484 1 1>}
  n50 = ScalarParameter() {defaultValue=-0.064, parameterName=Logo_Pos_x}
  n51 = ScalarParameter() {defaultValue=0.016, parameterName=Logo_Pos_y}
  n52 = Constant() {r=1}
  n53 = ScalarParameter() {defaultValue=0.75, parameterName=AspectRatioLogo}
  n54 = MaterialFunctionCall(
   UVs=Add(A=NamedRerouteUsage(), B=AppendVector(A=@n50, B=@n51))
   Texture Scale=Multiply(A=AppendVector(A=@n52, B=@n53), B=ScalarParameter() {defaultValue=1.015, parameterName=Logo_scale})
  ) {materialFunction=&ScaleUVsByCenter}
  n55 = Multiply(
   A=TextureSampleParameter2D(UVs=Clamp(None=@n54.UVs)) {parameterName=Banner, texture=/Game/UI/Login/Art/T_EpicLogo}.R
   B=ScalarParameter() {defaultValue=1, parameterName=LogoEnabled}
  )
  NamedRerouteDeclaration(None=StaticSwitchParameter(True=Constant() {r=1}, False=@n55) {desc=Disable logo on characters, parameterName=Use3Dto2DScreenWPO}) {name=Logo, nodeColor=<rgba 0.9688 1 0 1>}
  n56 = Multiply(A=NamedRerouteUsage(), B=Constant2Vector() {g=1, r=0.5})
  n57 = Ceil(None=Multiply(A=Time(), B=ScalarParameter() {defaultValue=5, parameterName=Glitch Speed}))
  n58 = ScalarParameter() {defaultValue=33.33, parameterName=GlitchRandomOffset}
  n59 = AppendVector(A=Frac(None=Multiply(A=@n57, B=@n58)), B=Constant())
  NamedRerouteDeclaration(None=TextureSample(UVs=Add(A=@n56, B=@n59)) {samplerType=SAMPLERTYPE_Masks, texture=/Game/Environments/TechArt/Hologram/T_GlitchNoise}.R) {name=DistortionDamage, nodeColor=<rgba 0.7109 0 1 1>}
  n60 = CollectionParameter() {collection=&MPC_Hologram, parameterName=ColorSwitch}
  n61 = NamedRerouteDeclaration(None=CollectionParameter() {collection=&MPC_Hologram, parameterName=Captured}) {name=Captured, nodeColor=<rgba 0.9219 1 0 1>}
  NamedRerouteDeclaration(None=Saturate(None=Add(A=@n60, B=@n61))) {name=ColorSwitch, nodeColor=<rgba 0 0.2266 1 1>}
  NamedRerouteDeclaration(None=CollectionParameter() {collection=&MPC_Hologram, parameterName=ScreensEnabled}) {name=ScreensEnabled, nodeColor=<rgba 0.6875 0 1 1>}
  NamedRerouteDeclaration(None=CollectionParameter() {collection=&MPC_Hologram, parameterName=Contested}) {name=Contested, nodeColor=<rgba 0.6875 0 1 1>}
  NamedRerouteDeclaration(None=CollectionParameter() {collection=&MPC_Hologram, parameterName=CaptureAnimPhase}) {name=CaptureANimPhase, nodeColor=<rgba 1 0 0.5859 1>}
  n62 = ScalarParameter() {defaultValue=0.95, parameterName=LogoMin}
  NamedRerouteDeclaration(None=StaticSwitchParameter(True=Constant(), False=Multiply(A=@n55, B=@n62)) {desc=Disable logo on characters, parameterName=Use3Dto2DScreenWPO}) {name=LogoFinal, nodeColor=<rgba 1 0 0.2109 1>}
  n63 = NamedRerouteDeclaration(None=Ceil(None=Saturate(None=@n1))) {name=Progress, nodeColor=<rgba 1 0 0.1641 1>}
  n64 = ScalarParameter() {defaultValue=0.9924, parameterName=ProgressLine_Body_Opacity}
  n65 = Power(Base=ComponentMask(None=@n1) {g=true}, Exp=ScalarParameter() {defaultValue=0.0601, parameterName=StripExponent_Anim})
  n66 = ScalarParameter() {defaultValue=0.232, parameterName=Progress_AnimSpeed}
  n67 = AppendVector(A=@n66, B=@n66)
  n68 = LinearInterpolate(A=@n67, B=Multiply(A=@n67) {constB=-1}, Alpha=ScalarParameter() {parameterName=AnimDirection})
  n69 = Multiply(
   A=ComponentMask(None=Panner(Coordinate=@n65, Speed=@n68)) {r=true}
   B=ScalarParameter() {defaultValue=13.2593, parameterName=StripsDistance_Anim}
  )
  n70 = Multiply(A=Saturate(None=Sine(None=@n69)), B=ScalarParameter() {defaultValue=0.2845, parameterName=StripsDistance_Anim_Brightness})
  NamedRerouteDeclaration(None=Add(A=Multiply(A=@n63, B=@n64) {constB=0.5}, B=Multiply(A=@n63, B=@n70))) {name=ProgressBar With Line, nodeColor=<rgba 0 1 0.1562 1>}
  OneMinus(None=@n49)
  n71 = Multiply(A=Time(), B=ScalarParameter() {defaultValue=0.048, parameterName=LogoSpeed_1})
  Rotator(Coordinate=@n54.UVs, Time=Multiply(A=Sine(None=@n71), B=ScalarParameter() {defaultValue=0.55, parameterName=LogoRotationMax}))
  ScalarParameter() {defaultValue=0.3405, parameterName=ProgressNormalized}
  Sign(None=@n1)
 ]}
