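# Opaque Default Lit base material for screen surfaces, built around the MF_ScreenProgress function, which by name adds a progress display.
# SetMaterialAttributes builds a base layer from a dark ScreensColor parameter and a Roughness parameter, then feeds it into that function.
# The function's UVs, Mask and IsEnabled inputs are unconnected, so its internal defaults apply.
# The material supports Nanite and is flagged as having pixel animation.
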Material /Game/Environments/TechArt/ScreenProgress/M_ScreenProgress
graph {
 properties {
  bAutomaticallySetUsageInEditor=false
  bHasPixelAnimation=true
  bUseMaterialAttributes=true
  bUsedWithNanite=true
  pixelDepthOffsetMode=PDOM_Legacy
  usedShadingModels=Default Lit
 }
 nodes [MaterialFunctionCall!(
   In=SetMaterialAttributes!(
    MaterialAttributes
    Base Color=VectorParameter() {parameterName=ScreensColor, defaultValue=<rgba 0.0104 0.0104 0.0104 1>}.RGB
    Roughness=ScalarParameter() {defaultValue=0.656, parameterName=Roughness}
   ) {attributeSetTypes [#1, #2]}
   UVs
   Mask
   IsEnabled
  ) {materialFunction=&/MF_ScreenProgress}]
}
